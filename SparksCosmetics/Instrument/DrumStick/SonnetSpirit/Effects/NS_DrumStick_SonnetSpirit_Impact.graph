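# Niagara impact effect for the SonnetSpirit drumstick cosmetic, a short blue burst of flash, sparks, toon smoke, flickering sparkles and surface debris.
# FN_PhysMatDebris uses User.PhysicalSurface through FN_PhysicalSurface to drive per-surface velocity, cone angle and gravity.
# Flash layers meshes, sprites and a light in one emitter, switching values by spawn group.
# Time-of-day tints and front-end modifiers adjust color or opacity on several layers, and sparks thin out with distance.
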NiagaraSystem /SparksCosmetics/Instrument/DrumStick/SonnetSpirit/Effects/NS_DrumStick_SonnetSpirit_Impact
dependencies {
 usedDataInterfaces [
  &NiagaraDataInterfaceCurve
  &NiagaraDataInterfaceColorCurve
  &NiagaraDataInterfaceMeshRendererInfo
 ]
 usedDynamicInputs [&MakeCustomFloatFromBool]
 usedModules [
  &self:SystemSpawnScript.SystemScriptSource.SystemScriptGraph.NiagaraNodeAssignment_0.SetVariables_#1
  &SystemState
 ]
}
emitters {
 FN_PhysMatDebris {
  emitterUpdateScript [
   &EmitterState()
   &SpawnBurst_Instantaneous(
    Spawn Count=&RandomRangeInt(Minimum=5, Maximum=10)
    Use Spawn Probability=false
    UseLoopCountLimit=false
   )
  ]
  particleSpawnScript [
   /SetVariables(Particles.BiasedRandom=&FloatFromCurve(
     FloatCurve=DataInterface() {propertyValues {Curve {keys [
         <at 0.2 1 RCIM_Cubic>
         {arriveTangent=-3.3394, interpMode=RCIM_Cubic, leaveTangent=-3.3394, tangentMode=RCTM_User, time=0.5, value=0.5}
         <at 0.8 0 RCIM_Cubic>
        ]}}}
     CurveIndex=&RandomRangeFloat(Minimum=0, Maximum=1)
    )) {isSetParameters=true}
   &InitializeParticle(
    Color Mode=::NewEnumerator0
    Lifetime=&RandomRangeFloat(Minimum=0.5, Maximum=1.25)
    Mass=&Lerp_Float(A=0.6, B=1.2, Alpha=$Particles.BiasedRandom)
    Mass Mode=::NewEnumerator1
    Mesh Scale=&Multiply_VectorByFloat(
     Vector=&RandomRangeVector(Minimum=<0.9 0.4 0.5>, Maximum=<1 1 1>)
     Float=&Lerp_Float(A=0.7, B=1, Alpha=$Particles.BiasedRandom)
    )
    Mesh Scale Mode=::NewEnumerator4
    UsePositionOffset=false
    Write Material Random=true
   )
   &FN_PhysicalSurface(
    Dirt Mesh Count=2
    Glass Mesh Count=3
    Ice Mesh Count=2
    Metal Gravity=&RandomRangeFloat(Minimum=-350, Maximum=-1800)
    Metal Mesh Count=2
    Metal Velocity=&RandomRangeFloat(Minimum=80, Maximum=375)
    Metal Velocity Cone Angle=130
    Physical Surface=$User.PhysicalSurface
    Stone Gravity=&RandomRangeFloat(Minimum=-700, Maximum=-1250)
    Stone Mesh Count=2
    Stone Velocity=&RandomRangeFloat(Minimum=80, Maximum=550)
    Stone Velocity Cone Angle=120
    Wood Gravity=&RandomRangeFloat(Minimum=-450, Maximum=-900)
    Wood Mesh Count=2
    Wood Scale=1.5
    Wood Velocity=&RandomRangeFloat(Minimum=80, Maximum=600)
    Wood Velocity Cone Angle=130
   )
   &AddVelocityInCone(
    Cone Angle=$Output.FN_PhysicalSurface.SurfaceResponseConeAngle
    Module.Use Velocity Falloff On Cone Axis=false
    Velocity Strength=$Output.FN_PhysicalSurface.SurfaceResponseVelocity
   )
   &InitialMeshOrientation(Use Rotation Vector=false)
   &AddRotationalVelocity(Rotation Rate=&RandomVector(Vector Scale=&RandomRangeFloat(Minimum=1, Maximum=2)))
   &ApplyInitialForces()
   /SetVariables(Particles.GravityOverrideValue=$Output.FN_PhysicalSurface.GravityOverride) {isSetParameters=true}
  ]
  particleUpdateScript [
   &ParticleState()
   &GravityForce(Gravity=$Particles.GravityOverrideValue)
   &Collision(
    Advanced Aging Rate=1
    Age Colliding Particles=false
    CPU Collision Trace Channel=ECC_WorldStatic
    Collision Enabled=&Boolean_AndOperation(
     Bool1=&SetBoolByFloatComparison(A=$Particles.Initial.Lifetime, B=1)
     Bool2=&SetBoolByFloatComparison(A=$Particles.Age, B=0.12)
    )
    Control Roll On Collision=true
    Friction=1
    Friction During a Bounce=0.25
    Max CPU Trace Length=100
    Module.Control Rest State Via Surface Normal=false
    Module.EnableMaxCollisionCount=false
    Module.Randomize Collision Normal Vector=true
    Randomize Collision Normal=0.25
    ScaleRateOfRotation=false
   ) {enabled=false}
   &SolveForcesAndVelocity(Clamp Velocity=false, Limit Acceleration=false)
   &ScaleMeshSize(Scale Factor=&VectorFromFloat(Value=&FloatFromCurve(FloatCurve=DataInterface() {propertyValues {Curve {keys [<at 0.6 1 RCIM_Cubic>, {arriveTangent=-6.308, interpMode=RCIM_Cubic, leaveTangent=-6.308, tangentMode=RCTM_User, time=1, value=0}]}}})))
   &DynamicMaterialParameters(Module.Param0WriteEnabled=true, Module.Param1WriteEnabled=true, Module.Param2WriteEnabled=false, Module.Param3WriteEnabled=false)
  ]
  rendererClasses [&NiagaraSpriteRendererProperties, &NiagaraMeshRendererProperties]
  renderers [
   {rendererClass=&NiagaraSpriteRendererProperties}
   {rendererClass=&NiagaraMeshRendererProperties, rendererIndex=1}
   {rendererClass=&NiagaraMeshRendererProperties, rendererIndex=2}
   {rendererClass=&NiagaraMeshRendererProperties, rendererIndex=3}
   {rendererClass=&NiagaraMeshRendererProperties, rendererIndex=4}
   {rendererClass=&NiagaraMeshRendererProperties, rendererIndex=5}
   {rendererClass=&NiagaraMeshRendererProperties, rendererIndex=6}
   {rendererClass=&NiagaraMeshRendererProperties, rendererIndex=7}
   {rendererClass=&NiagaraMeshRendererProperties, rendererIndex=8}
  ]
 }
 Flash {
  emitterUpdateScript [
   &EmitterState()
   &SpawnBurst_Instantaneous(Spawn Count=1, Use Spawn Probability=false)
   &SpawnBurst_Instantaneous(Spawn Count=&RandomRangeInt(Minimum=9, Maximum=15), Spawn Group=1, Use Spawn Probability=false) {module=SpawnBurst_Instantaneous001}
   &SpawnBurst_Instantaneous(Spawn Count=5, Spawn Group=2, Use Spawn Probability=false, UseLoopCountLimit=false) {enabled=false, module=SpawnBurst_Instantaneous002}
   &SpawnBurst_Instantaneous(Spawn Group=3, Use Spawn Probability=false, UseLoopCountLimit=false) {module=SpawnBurst_Instantaneous003}
   &SpawnBurst_Instantaneous(Spawn Group=4, Use Spawn Probability=false, UseLoopCountLimit=false) {module=SpawnBurst_Instantaneous004}
   &SpawnBurst_Instantaneous(Spawn Group=5, Use Spawn Probability=false, UseLoopCountLimit=false) {module=SpawnBurst_Instantaneous005}
  ]
  particleSpawnScript [
   /SetVariables(Particles.VisibilityTag=&SpawnGroup()) {isSetParameters=true}
   &InitializeParticle(
    Color=&MergeHelper_ColorSwitch(
     0=<rgba 1.7708 3.0056 4 1>
     1=<rgba 0 0.2159 0.807 1>
     2=<rgba 0 0.2157 0.8047 1>
     3=<rgba 0.1114 0.7694 1 1>
     4=<rgba 1 1 1 1>
     5=<rgba 0.0122 0.1058 0.224 1>
     Index=&SpawnGroup()
     Num=::NewEnumerator4
    )
    Lifetime=&MergeHelper_FloatSwitch(
     0=0.23
     1=&RandomRangeFloat(Minimum=0.1, Maximum=0.15)
     2=&RandomRangeFloat(Minimum=0.09, Maximum=0.12)
     3=0.2
     4=0.15
     5=0.18
     Index=&SpawnGroup()
     Num=::NewEnumerator4
    )
    Mesh Scale=&Multiply_VectorByFloat(Vector=&RandomRangeVector(Minimum=<1 2 2>, Maximum=<0.75 1.5 1.5>), Float=$System.GlobalScale)
    Mesh Scale Mode=::NewEnumerator4
    Sprite Size=&Multiply_Vector2D_ByFloat(
     Float=$System.GlobalScale
     Vector2D=&MergeHelper_Vector2Switch(
      0=<1 1>
      1=&RandomRangeVector2D(Minimum=<9 50>, Maximum=<15 60>)
      2=&RandomRangeVector2D(Minimum=<30 45>, Maximum=<45 60>)
      3=&Vector2DFromFloat(Value=175)
      4=<1 1>
      5=&Vector2DFromFloat(Value=125)
      Index=$Emitter.SpawnGroup
      Num=::NewEnumerator4
     )
    )
    Sprite Size Mode=::NewEnumerator4
    UsePositionOffset=false
    Write Material Random=true
   )
   &ShapeLocation(
    MaskSpawnGroup=false
    Override Seed=false
    Sphere Radius=&MergeHelper_FloatSwitch(
     0=0
     1=&Multiply_Float(A=6, B=$System.GlobalScale)
     2=&Multiply_Float(A=6, B=$System.GlobalScale)
     3=0
     4=0
     5=0
     Index=&SpawnGroup()
     Num=::NewEnumerator4
    )
    UseOwnerScale=true
   )
   &DynamicMaterialParameters(
    Index 0 Param 2=45
    Index 0 Param 3=1
    Index 0 Param 4=1
    Module.Param0WriteEnabled=false
    Module.Param1WriteEnabled=true
    Module.Param2WriteEnabled=true
    Module.Param3WriteEnabled=true
   ) {module=DynamicMaterialParameters001}
   &InitialMeshOrientation(
    Rotation=&RandomRangeVector(Minimum=<-1 -1 -1>, Maximum=<1 1 1>)
    Use Orientation Vector=true
    Use Rotation Vector=true
   )
   &CameraOffset(Camera Offset Amount=&MergeHelper_FloatSwitch(0=20, 1=10, 2=10, 3=20, 4=0, 5=20, Index=&SpawnGroup(), Num=::NewEnumerator4))
   &AddVelocityFromPoint(
    Module.Use Falloff Distance=false
    Velocity Strength=&MergeHelper_FloatSwitch(0=0, 1=100, 2=125, 3=0, 4=0, Index=&SpawnGroup(), Num=::NewEnumerator3)
   )
  ]
  particleUpdateScript [
   &ParticleState()
   &DynamicMaterialParameters(
    Index 0 Param 1=&FloatFromCurve(FloatCurve=DataInterface() {propertyValues {Curve {keys [
         {interpMode=RCIM_Cubic, leaveTangentWeight=0.4953, tangentMode=RCTM_User, time=0.5, value=0}
         {arriveTangent=5.0084, interpMode=RCIM_Cubic, leaveTangent=5.0084, tangentMode=RCTM_User, time=1, value=1}
        ]}}})
    Module.Index0Float0Write=true
    Module.Index0Float1Write=true
    Module.Index0Float2Write=true
    Module.Index0Float3Write=true
   )
   &ScaleMeshSize(Scale Factor=&VectorFromFloat(Value=&FloatFromCurve(
      FloatCurve=DataInterface() {propertyValues {Curve {
         defaultValue=0.5
         keys [
          {arriveTangent=1.7893, interpMode=RCIM_Cubic, leaveTangent=1.7893, tangentMode=RCTM_User, time=0, value=0}
          {
           arriveTangent=0.6795
           arriveTangentWeight=0.33
           interpMode=RCIM_Cubic
           leaveTangent=0.6795
           leaveTangentWeight=0.1414
           tangentMode=RCTM_User
           time=0.7
           value=0.8
          }
          {arriveTangent=1.8732, arriveTangentWeight=0.2191, interpMode=RCIM_Cubic, leaveTangent=1.8732, tangentMode=RCTM_User, time=1, value=1}
         ]
        }}}
      Scale Curve=0.5
     )))
   &ScaleColor(
    Scale Alpha=&MergeHelper_FloatSwitch(
     0=&FloatFromCurve(FloatCurve=DataInterface() {propertyValues {Curve {keys [<at 0 1>,<at 1 0>]}}})
     1=&FloatFromCurve(FloatCurve=DataInterface() {propertyValues {Curve {
         defaultValue=1
         keys [{interpMode=RCIM_Cubic, tangentMode=RCTM_User, time=0, value=1}, {interpMode=RCIM_Cubic, tangentMode=RCTM_User, time=1, value=0}]
        }}})
     2=&FloatFromCurve(FloatCurve=DataInterface() {propertyValues {Curve {
         defaultValue=1
         keys [
          {arriveTangent=-2.0249, interpMode=RCIM_Cubic, leaveTangent=-2.0249, tangentMode=RCTM_User, time=0, value=1}
          {interpMode=RCIM_Cubic, tangentMode=RCTM_User, time=1, value=0}
         ]
        }}})
     3=&FloatFromCurve(
      FloatCurve=DataInterface() {propertyValues {Curve {
         defaultValue=1
         keys [
          {interpMode=RCIM_Cubic, tangentMode=RCTM_User, time=0, value=1}
          {arriveTangent=-1.2989, interpMode=RCIM_Cubic, leaveTangent=-1.2989, tangentMode=RCTM_User, time=1, value=0}
         ]
        }}}
      Scale Curve=&FN_FrontEndFloatModifier(Gameplay=1, Front End=0.75)
     )
     4=1
     5=&FloatFromCurve(
      FloatCurve=DataInterface() {propertyValues {Curve {
         defaultValue=1
         keys [
          {interpMode=RCIM_Cubic, tangentMode=RCTM_User, time=0, value=1}
          {arriveTangent=-2.0745, interpMode=RCIM_Cubic, leaveTangent=-2.0745, tangentMode=RCTM_User, time=1, value=0}
         ]
        }}}
      Scale Curve=&FloatFromCurve(
       FloatCurve=DataInterface() {propertyValues {Curve {
          defaultValue=1
          keys [
           {interpMode=RCIM_Cubic, tangentMode=RCTM_User, time=0, value=0.45}
           {interpMode=RCIM_Cubic, tangentMode=RCTM_User, time=0.5, value=1.2}
           {tangentMode=RCTM_User, time=1, value=0.45}
          ]
         }}}
       CurveIndex=Linked() {linkedVariable {name=NPC.FortniteNPC.TimeOfDay_0-1, type=NiagaraFloat}}
       Scale Curve=&FN_FrontEndFloatModifier(Gameplay=1.1, Front End=0.7)
      )
     )
     Index=&SpawnGroup()
     Num=::NewEnumerator4
    )
    Scale RGB=&MergeHelper_VectorSwitch(
     0=&Multiply_VectorByFloat(
      Vector=&MakeVectorFromLinearColorRGB(Color=&ColorFromCurve(ColorCurve=DataInterface() {propertyValues {
          AlphaCurve {keys []}
          BlueCurve {keys [<at 0 5 RCIM_Cubic>,<at 0.2025 1>,<at 0.833 1 RCIM_Cubic>]}
          GreenCurve {keys [<at 0 4 RCIM_Cubic>,<at 0.2025 0.6953>,<at 0.833 0.6016 RCIM_Cubic>]}
          RedCurve {keys [<at 0 2 RCIM_Cubic>,<at 0.2025 0.3906>,<at 0.833 0.2031 RCIM_Cubic>]}
         }}))
      Float=&FloatFromCurve(FloatCurve=DataInterface() {propertyValues {Curve {keys [<at 0 4>,<at 0.2 1.5>,<at 1 1>]}}})
     )
     1=&MakeVectorFromLinearColorRGB(Color=&ColorFromCurve(ColorCurve=DataInterface() {propertyValues {
         AlphaCurve {keys []}
         BlueCurve {keys [<at 0 5 RCIM_Cubic>,<at 0.4312 1>,<at 1 1 RCIM_Cubic>]}
         GreenCurve {keys [<at 0 3.0166 RCIM_Cubic>,<at 0.4312 0.6953>,<at 1 0.6016 RCIM_Cubic>]}
         RedCurve {keys [<at 0 1.6033 RCIM_Cubic>,<at 0.4312 0.3906>,<at 1 0.2031 RCIM_Cubic>]}
        }}))
     2=&VectorFromFloat(Value=0.4)
     3=&VectorFromFloat(Value=&FloatFromCurve(
       FloatCurve=DataInterface() {propertyValues {Curve {keys [<at 0 4>,<at 1 1>]}}}
       Scale Curve=&FN_FrontEndFloatModifier(Gameplay=1, Front End=0.9)
      ))
     4=<1 1 1>
     5=<1 1 1>
     Index=&SpawnGroup()
     Num=::NewEnumerator4
    )
    ScaleA=true
    ScaleRGB=true
   )
   &ScaleSpriteSize(
    Scale Factor=&MergeHelper_Vector2Switch(
     0=<0 0>
     1=&MakeVector2D(
      X=&FloatFromCurve(FloatCurve=DataInterface() {propertyValues {Curve {
          defaultValue=1
          keys [
           {interpMode=RCIM_Cubic, tangentMode=RCTM_User, time=0, value=0}
           {arriveTangentWeight=0.2236, interpMode=RCIM_Cubic, leaveTangentWeight=0.1491, tangentMode=RCTM_User, time=0.7, value=1}
           {arriveTangent=-2.2754, interpMode=RCIM_Cubic, leaveTangent=-2.2754, tangentMode=RCTM_User, time=1, value=0.5}
          ]
         }}})
      Y=&FloatFromCurve(FloatCurve=DataInterface() {propertyValues {Curve {
          defaultValue=1
          keys [
           {arriveTangent=3.143, interpMode=RCIM_Cubic, leaveTangent=3.143, tangentMode=RCTM_User, time=0, value=0}
           {arriveTangent=0.8537, interpMode=RCIM_Cubic, leaveTangent=0.8537, tangentMode=RCTM_User, time=1, value=1}
          ]
         }}})
     )
     2=&Vector2DFromFloat(Value=&FloatFromCurve(FloatCurve=DataInterface() {propertyValues {Curve {
          defaultValue=1
          keys [
           {arriveTangent=0.9682, interpMode=RCIM_Cubic, leaveTangent=0.9682, tangentMode=RCTM_User, time=0, value=0}
           {interpMode=RCIM_Cubic, tangentMode=RCTM_User, time=1, value=1}
          ]
         }}}))
     3=&Multiply_Vector2D_ByFloat(
      Float=&FloatFromCurve(FloatCurve=DataInterface() {propertyValues {Curve {
          defaultValue=0
          keys [
           {interpMode=RCIM_Cubic, tangentMode=RCTM_User, time=0, value=1}
           {arriveTangent=-1.4808, interpMode=RCIM_Cubic, leaveTangent=-1.4808, tangentMode=RCTM_User, time=1, value=0}
          ]
         }}})
      Vector2D=<1 1>
     )
     4=<0 0>
     5=&Vector2DFromFloat(Value=&FloatFromCurve(FloatCurve=DataInterface() {propertyValues {Curve {
          defaultValue=0
          keys [
           {arriveTangent=2.4894, interpMode=RCIM_Cubic, leaveTangent=2.4894, tangentMode=RCTM_User, time=0, value=0}
           {interpMode=RCIM_Cubic, tangentMode=RCTM_User, time=1, value=1}
          ]
         }}}))
     Index=&SpawnGroup()
     Num=::NewEnumerator4
    )
    Scale Sprite Size Mode=::NewEnumerator2
   )
   &SolveForcesAndVelocity(Clamp Velocity=false, Limit Acceleration=false)
   /SetVariables(Particles.Flare Dyn 1:Vector4f=&MakeVector4(
     X=&FloatFromCurve(FloatCurve=DataInterface() {propertyValues {Curve {keys [<at 0 1 RCIM_Cubic>,<at 0.8 0 RCIM_Cubic>]}}})
     Y=2
     Z=1
     W=1
    )) {isSetParameters=true}
   &FN_Light_Attributes(
    Color=<rgba 0 0.4378 1 1>
    Create Color Attribute=true
    Create Exponent Attribute=true
    Create Radius Attribute=true
    Exponent=3
    Radius=95
    Scale Color=&FloatFromCurve(
     FloatCurve=DataInterface() {propertyValues {Curve {
        defaultValue=1
        keys [{interpMode=RCIM_Cubic, tangentMode=RCTM_User, time=0, value=1}, {interpMode=RCIM_Cubic, tangentMode=RCTM_User, time=1, value=0}]
       }}}
     Scale Curve=12
    )
   )
   /SetVariables(
    Particles.Outline Color:LinearColor=<rgba 0 0.504 0.9609 1>
    Particles.OutlineScale:Vector3f=&Multiply_VectorByFloat(Vector=$Particles.Scale, Float=1.65)
   ) {isSetParameters=true}
  ]
  rendererClasses [
   &NiagaraMeshRendererProperties
   &NiagaraSpriteRendererProperties
   &NiagaraLightRendererProperties
  ]
  renderers [
   {rendererClass=&NiagaraMeshRendererProperties, rendererIndex=0}
   {rendererClass=&NiagaraSpriteRendererProperties, rendererIndex=1}
   {rendererClass=&NiagaraSpriteRendererProperties, rendererIndex=2}
   {rendererClass=&NiagaraSpriteRendererProperties, rendererIndex=3}
   {rendererClass=&NiagaraLightRendererProperties, rendererIndex=4}
   {rendererClass=&NiagaraSpriteRendererProperties, rendererIndex=5}
   {rendererClass=&NiagaraMeshRendererProperties, rendererIndex=6}
  ]
 }
 Snow_Sparkles001 {
  emitterUpdateScript [
   &EmitterState(Life Cycle Mode=::NewEnumerator1, Loop Behavior=::NewEnumerator1, Loop Duration=1, UseLoopDelay=false)
   &SpawnBurst_Instantaneous(
    Spawn Count=&RandomRangeInt(Minimum=4, Maximum=6)
    Use Spawn Probability=false
    UseLoopCountLimit=false
   )
  ]
  particleSpawnScript [
   &InitializeParticle(
    Color=&MultiplyLinearColor(
     Linear Color=&RandomRangeLinearColor(Minimum=<rgba 0.2806 0.3757 1 0.25>, Maximum=<rgba 0.153 0.6002 1 0.25>)
     Scale Factor=&MakeVector4_FromLinearColor(LinearColor=$System.TOD Color)
    )
    Lifetime=&RandomRangeFloat(Minimum=0.4, Maximum=1)
    Sprite Size Mode=::NewEnumerator3
    Uniform Sprite Size=&Multiply_Float(A=&RandomRangeFloat(Minimum=3, Maximum=9), B=$System.GlobalScale)
    UsePositionOffset=false
    Write Material Random=true
   )
   &ShapeLocation(
    MaskSpawnGroup=false
    Sphere Distribution=::NewEnumerator3
    Sphere Radius=&Multiply_Float(A=5, B=$System.GlobalScale)
    UseOwnerScale=true
   ) {module=ShapeLocation001}
   &AddVelocity(
    Module.Use Falloff Distance=false
    Origin Offset=&FN_FrontEndVector3Modifier(Gameplay=<-2 0 0>, Front End=<0 0 0>)
    Override Seed=false
    Velocity Mode=::NewEnumerator1
    Velocity Speed=&Multiply_Float(A=&RandomRangeFloat(Minimum=75, Maximum=175), B=$System.GlobalScale)
   )
   &CameraOffset(Camera Offset Amount=15)
  ]
  particleUpdateScript [
   &ParticleState()
   /SetVariables(Particles.Flicker:NiagaraFloat=&Waveform(
     Global Phase=&Add_Float(A=&RandomRangeFloat(Minimum=0, Maximum=5), B=$Engine.Time)
     "Waveform [1]"=::NewEnumerator12
     "[1] Frequency"=5
     "[1] Amplitude Min/Max"=<0.75 5>
    )) {isSetParameters=true}
   &ScaleSpriteSize(
    Uniform Curve Scale=/Game/Effects/Niagara/DynamicInputs/Math/FN_Fit(
     Clamp:NiagaraBool=false
     Value:NiagaraFloat=$Particles.Flicker
     SourceMin:NiagaraFloat=0.75
     SourceMax:NiagaraFloat=5
     DestMin:NiagaraFloat=0.25
     DestMax:NiagaraFloat=2
    )
    Uniform Curve Sprite Scale=DataInterface() {propertyValues {Curve {
       defaultValue=0
       keys [
        {interpMode=RCIM_Cubic, leaveTangentWeight=0.5801, tangentMode=RCTM_User, time=0, value=0}
        {arriveTangentWeight=0.0667, interpMode=RCIM_Cubic, leaveTangentWeight=0.0996, tangentMode=RCTM_User, time=0.2, value=0.9847}
        {arriveTangentWeight=0.158, interpMode=RCIM_Cubic, leaveTangentWeight=0.2649, tangentMode=RCTM_User, time=0.5, value=1}
        {arriveTangent=-3.5183, arriveTangentWeight=0.4104, interpMode=RCIM_Cubic, leaveTangent=-3.5183, tangentMode=RCTM_User, time=1, value=0}
       ]
      }}}
   )
   &Drag(
    Drag=&Multiply_Float(
     A=&RandomRangeFloat(Minimum=1, Maximum=5)
     B=&FloatFromCurve(FloatCurve=DataInterface() {propertyValues {Curve {defaultValue=1, keys [{tangentMode=RCTM_User, time=0.2, value=2}, {tangentMode=RCTM_User, time=0.65, value=0.5}]}}})
    )
    Use Linear Drag=true
    Use Rotational Drag=true
   )
   &CurlNoiseForce(
    Bias Noise=false
    Noise Frequency=150
    Noise Strength=&RandomRangeFloat(Minimum=80, Maximum=100)
    Pan Noise=false
   )
   &GravityForce(Gravity=&Multiply_VectorByFloat(
     Vector=<0 0 -200>
     Float=&FloatFromCurve(
      FloatCurve=DataInterface() {propertyValues {Curve {
         defaultValue=1
         keys [
          {arriveTangent=1.9398, interpMode=RCIM_Cubic, leaveTangent=1.9398, tangentMode=RCTM_User, time=0.15, value=0}
          {interpMode=RCIM_Cubic, tangentMode=RCTM_User, time=1, value=1}
         ]
        }}}
      Scale Curve=&RandomRangeFloat(Minimum=0.5, Maximum=1)
     )
    ))
   &SolveForcesAndVelocity(Clamp Velocity=false, Limit Acceleration=false)
   &ScaleColor(
    Scale Alpha=&FloatFromCurve(
     FloatCurve=DataInterface() {propertyValues {Curve {
        defaultValue=1
        keys [
         <at 0 0.5 RCIM_Cubic>
         {interpMode=RCIM_Cubic, tangentMode=RCTM_User, time=0.2, value=1}
         {arriveTangentWeight=0.3564, interpMode=RCIM_Cubic, tangentMode=RCTM_User, time=1, value=0}
        ]
       }}}
     Scale Curve=&FN_FrontEndFloatModifier(Gameplay=1, Front End=0.5)
    )
    Scale RGB=&Multiply_VectorByFloat(
     Vector=&MakeVectorFromLinearColorRGB(Color=&ColorFromCurve(ColorCurve=DataInterface() {propertyValues {
         AlphaCurve {keys [<at 0 1 RCIM_Cubic>]}
         BlueCurve {keys [<at 0.0556 1.25 RCIM_Cubic>,<at 0.5289 1>,<at 1 1 RCIM_Cubic>]}
         GreenCurve {keys [<at 0.0556 0.6191 RCIM_Cubic>,<at 0.5289 0.5227>,<at 1 0.3346 RCIM_Cubic>]}
         RedCurve {keys [<at 0.0556 0.2353 RCIM_Cubic>,<at 0.5289 0.2433>,<at 1 0.1578 RCIM_Cubic>]}
        }}))
     Float=$Particles.Flicker
    )
    ScaleA=true
    ScaleRGB=true
   )
  ]
  rendererClasses [&NiagaraSpriteRendererProperties]
  renderers [{rendererClass=&NiagaraSpriteRendererProperties}]
 }
 Sparks {
  emitterUpdateScript [
   &EmitterState(
    Enable Distance Culling=true
    Max Distance Response=::KillImmediately
    Module.bApplySpawnCountScaleByDistance=true
    Module.bUseMaxDistance=true
    Module.bUseMinDistance=false
    Scalability Mode=::NewEnumerator1
    Scale Spawn Count=true
    Spawn Count Scale By Distance=DataInterface() {propertyValues {Curve {keys [<at 0.3676 0.9657>,<at 0.9962 0.4982>]}}}
   )
   &SpawnBurst_Instantaneous(Spawn Count=&UniformRangedInt(Minimum=12, Maximum=15), Use Spawn Probability=false)
  ]
  particleSpawnScript [
   &InitializeParticle(
    Color=<rgba 0 0.7689 1.5 1>
    Lifetime Max=0.45
    Lifetime Min=0.25
    Lifetime Mode=::NewEnumerator1
    Sprite Size Mode=::NewEnumerator3
    Uniform Sprite Size=&Multiply_Float(
     A=$System.GlobalScale
     B=&FloatFromCurve(
      FloatCurve=DataInterface() {propertyValues {Curve {keys [<at 0.1 9 RCIM_Cubic>, {arriveTangent=-4.9067, interpMode=RCIM_Cubic, leaveTangent=-4.9067, tangentMode=RCTM_User, time=1, value=4}]}}}
      CurveIndex=&ReturnNormalizedExecIndex()
     )
    )
    UsePositionOffset=false
    Write Material Random=true
   )
   &ShapeLocation(
    MaskSpawnGroup=false
    Sphere Distribution=::NewEnumerator3
    Sphere Radius=&Multiply_Float(A=12, B=$System.GlobalScale)
    UseOwnerScale=true
   )
   &AddVelocity(
    Module.Use Falloff Distance=false
    Override Seed=false
    Velocity Mode=::NewEnumerator1
    Velocity Speed=&Multiply_Float(A=&RandomRangeFloat(Minimum=300, Maximum=700), B=$System.GlobalScale)
   )
   &DynamicMaterialParameters(
    Index 0 Param 1=3
    Index 0 Param 2=20
    Index 0 Param 3=1
    Module.Index0Float0Write=true
    Module.Index0Float1Write=true
    Module.Index0Float2Write=true
    Module.Index0Float3Write=true
   )
  ]
  particleUpdateScript [
   &ParticleState()
   &ScaleSpriteSize(Uniform Curve Sprite Scale=DataInterface() {propertyValues {Curve {
       defaultValue=0
       keys [
        {arriveTangent=9.0725, interpMode=RCIM_Cubic, leaveTangent=9.0725, tangentMode=RCTM_User, time=0, value=0}
        {arriveTangentWeight=0.1274, interpMode=RCIM_Cubic, leaveTangentWeight=0.7778, tangentMode=RCTM_User, time=0.2, value=0.9847}
        {arriveTangent=-0.0405, interpMode=RCIM_Cubic, leaveTangent=-0.0405, tangentMode=RCTM_User, time=1, value=0}
       ]
      }}})
   &Drag(Drag=&UniformRangedFloat(Minimum=12, Maximum=15), Use Linear Drag=true, Use Rotational Drag=false)
   &SolveForcesAndVelocity(Clamp Velocity=false, Limit Acceleration=false)
   &ScaleSpriteSizeBySpeed(Max Scale Factor=<0.75 5>, Min Scale Factor=<0.5 1.5>, Velocity Threshold=666)
   &ScaleColor(
    Linear Color Curve=DataInterface() {propertyValues {
      AlphaCurve {keys [<at 0.333 1>,<at 1 0>]}
      BlueCurve {keys [<at 0 6>,<at 0.166 5>,<at 0.333 4>,<at 0.9054 3>]}
      GreenCurve {keys [<at 0 4.8228>,<at 0.166 2.2711>,<at 0.333 1.1717>,<at 0.9054 0.295>]}
      RedCurve {keys [<at 0 0>,<at 0.166 0.0044>,<at 0.333 0>,<at 0.9054 0>]}
     }}
    Scale Mode=::NewEnumerator2
   )
  ]
  rendererClasses [&NiagaraSpriteRendererProperties]
  renderers [{rendererClass=&NiagaraSpriteRendererProperties}]
 }
 Toon_Smoke {
  emitterUpdateScript [
   &EmitterState(Life Cycle Mode=::NewEnumerator1, Loop Behavior=::NewEnumerator1, Loop Duration=1, UseLoopDelay=false)
   &SpawnBurst_Instantaneous(Spawn Count=2, Use Spawn Probability=false)
  ]
  particleSpawnScript [
   &CameraOffset(Camera Offset Amount=30)
   &DynamicMaterialParameters(
    Index 0 Param 2=&UniformAOrBFloat(A=0, Evaluation Type=::NewEnumerator0, Recalculate Random Each Loop=true, Override Seed=false)
    Module.Index0Float0Write=true
    Module.Index0Float1Write=true
    Module.Index0Float2Write=true
    Module.Index0Float3Write=true
   ) {module=DynamicMaterialParameters001}
   &InitializeParticle(
    Color=&MultiplyLinearColor(
     Linear Color=$System.TOD Color
     Scale Factor=&MakeVector4_FromLinearColor(LinearColor=&ColorFromCurve(
       ColorCurve=DataInterface() {propertyValues {
         AlphaCurve {keys [<at 0.0233 0.75 RCIM_Cubic>,<at 0.9834 1 RCIM_Cubic>]}
         BlueCurve {keys [<at 0 1 RCIM_Cubic>,<at 1 1.5 RCIM_Cubic>]}
         GreenCurve {keys [<at 0 1 RCIM_Cubic>,<at 1 1.5 RCIM_Cubic>]}
         RedCurve {keys [<at 0 1 RCIM_Cubic>,<at 1 1.5 RCIM_Cubic>]}
        }}
       CurveIndex=&ReturnNormalizedExecIndex()
      ))
    )
    Lifetime=&FloatFromCurve(
     FloatCurve=DataInterface() {propertyValues {Curve {keys [<at 0 1>,<at 1 0.65>]}}}
     CurveIndex=&ReturnNormalizedExecIndex()
     Scale Curve=0.65
    )
    Sprite Rotation Mode=::NewEnumerator2
    Sprite Size Mode=::NewEnumerator3
    Sprite UV Mode=::NewEnumerator3
    Uniform Sprite Size=&Multiply_Float(
     A=&FloatFromCurve(
      FloatCurve=DataInterface() {propertyValues {Curve {keys [<at 0 1>,<at 1 0.6>]}}}
      CurveIndex=&ReturnNormalizedExecIndex()
      Scale Curve=105
     )
     B=$System.GlobalScale
    )
    UsePositionOffset=false
    Write Material Random=true
   ) {module=InitializeParticle001}
  ]
  particleUpdateScript [
   &ParticleState()
   &ScaleSpriteSize(Uniform Curve Sprite Scale=DataInterface() {propertyValues {Curve {keys [
        {arriveTangent=0.6997, interpMode=RCIM_Cubic, leaveTangent=0.6997, tangentMode=RCTM_User, time=0, value=0.7}
        {arriveTangentWeight=0.3042, interpMode=RCIM_Cubic, leaveTangentWeight=0.0845, tangentMode=RCTM_User, time=0.7827, value=0.9719}
        {arriveTangent=2.7371, interpMode=RCIM_Cubic, leaveTangent=2.7371, tangentMode=RCTM_User, time=1, value=1.2153}
       ]}}})
   &SolveForcesAndVelocity(Clamp Velocity=false, Limit Acceleration=false)
   &DynamicMaterialParameters(
    Index 0 Param 1=&FloatFromCurve(FloatCurve=DataInterface() {propertyValues {Curve {keys [
         {arriveTangent=3.5093, interpMode=RCIM_Cubic, leaveTangent=3.5093, tangentMode=RCTM_User, time=0, value=0}
         {arriveTangent=0.6664, interpMode=RCIM_Cubic, leaveTangent=0.6664, tangentMode=RCTM_User, time=0.4875, value=0.7253}
         {arriveTangentWeight=0.1717, interpMode=RCIM_Cubic, tangentMode=RCTM_User, time=1, value=1}
        ]}}})
    Index 0 Param 2=1
    Module.Param0WriteEnabled=true
    Module.Param1WriteEnabled=true
    Module.Param2WriteEnabled=true
    Module.Param3WriteEnabled=true
   )
   &ScaleColor(
    Scale Alpha=&FloatFromCurve(
     FloatCurve=DataInterface() {propertyValues {Curve {keys [
         <at 0 1 RCIM_Cubic>
         {arriveTangent=-2.2323, arriveTangentWeight=0.3763, interpMode=RCIM_Cubic, leaveTangent=-2.2323, tangentMode=RCTM_User, time=1, value=0}
        ]}}}
     Scale Curve=&FN_FrontEndFloatModifier(Gameplay=0.8, Front End=0.5)
    )
    Scale RGB=&Multiply_Vector(
     A=&MakeVectorFromLinearColorRGB(Color=&ColorFromCurve(ColorCurve=DataInterface() {propertyValues {
         AlphaCurve {keys []}
         BlueCurve {keys [<at 0 3>,<at 1 2>]}
         GreenCurve {keys [<at 0 2.3264>,<at 1 1.4798>]}
         RedCurve {keys [<at 0 1.0352>,<at 1 1.0521>]}
        }}))
     B=&MakeVectorFromLinearColorRGB(Color=$System.TOD Color)
    )
    ScaleA=true
    ScaleRGB=true
   )
  ]
  rendererClasses [&NiagaraSpriteRendererProperties]
  renderers [{rendererClass=&NiagaraSpriteRendererProperties}]
 }
}
summary {
 systemName=NS_DrumStick_SonnetSpirit_Impact
 emitters [
  {
   emitterName=FN_PhysMatDebris
   rendererClasses [&NiagaraSpriteRendererProperties, &NiagaraMeshRendererProperties]
  }
  {emitterName=Toon_Smoke, rendererClasses [&NiagaraSpriteRendererProperties]}
  {
   emitterName=Flash
   rendererClasses [
    &NiagaraMeshRendererProperties
    &NiagaraSpriteRendererProperties
    &NiagaraLightRendererProperties
   ]
  }
  {emitterName=Sparks, rendererClasses [&NiagaraSpriteRendererProperties]}
  {emitterName=Snow_Sparkles001, rendererClasses [&NiagaraSpriteRendererProperties]}
 ]
 userVariables [
  User.bIsFrontEnd {type=NiagaraBool}
  User.bIsFrontEndPreview {type=NiagaraBool}
  User.PhysicalSurface {type=NiagaraInt32}
 ]
}
